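; Spline editing tool Blueprint that converts the points of its spline into a list of normalized radial distances.
; GeneratePoints loops over the spline points, takes each local position's 2D distance from the origin, divides it by BaseRadius and returns the resulting array.
; The GeneratePointsList custom event, construction script and standard event handlers are empty, so nothing in this graph calls the function.
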(event EventBeginPlay)

(event Collision|EventActorBeginOverlap
  (params (OtherActor :type Actor%20Object%20Reference)))

(event EventTick
  (params (DeltaSeconds :type Float%20%28single-precision%29)))

(event Custom|GeneratePointsList)

(fn GeneratePoints
  (for _index (range (Spline|GetNumberOfSplinePoints (Variables|Default|GetSpline)))
    (Utilities|Array|Add (LocalVariables|GetReturnPoints) (/ (Math|Vector|Distance2D(Vector) (.Position (Utilities|Struct|BreakSplinePoint (Spline|GetSplinePointAt (Variables|Default|GetSpline) _index "Local")))) (Variables|Default|GetBaseRadius))))
  (return (LocalVariables|GetReturnPoints)))

(fn ConstructionScript)
