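# Editor utility widget that bakes YokeHopUnkind control point radius scales from a spline editor Blueprint into a control points data asset.
# One button stores the Content Browser selection as the source and shows its class name.
# The output button compiles and saves that Blueprint, spawns its generated class, calls GeneratePoints, then destroys the actor.
# The resulting float array overwrites ControlPointRadiusScales on the picked data asset, which is then saved.
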EditorUtilityWidgetBlueprint /YokeHopUnkind/Tools/EUW_YokeHopKeeper_ControlPoints
graphs [
 EventGraph {nodes [
   "On Clicked (OutputButton)"() {is=title K2Node_ComponentBoundEvent}
   > n1 = GetBlueprintAsset(Object=Get SelectedObject() -> (SelectedObject:Object Reference))
   > CompileBlueprint!(execute, Blueprint:Blueprint Object Reference=@n1) -> (then, ReturnValue:bool)
   > SaveLoadedAsset(AssetToSave=@n1, bOnlyIfIsDirty=true)
   > n2 = GetSplineEditorClassFromSelectedObject!(execute, self, Object:Object Reference=Get SelectedObject()) -> (then, AsBP Yoke Hop Unkind Spline Editor:BP Yoke Hop Unkind Spline Editor Class Reference)
   > Branch(Condition=IsValidClass!(Class:Object Class Reference=@n2) -> (ReturnValue:bool))
   > n3 = SpawnActorFromClass!(execute, ActorClass:Actor Class Reference=@n2, Location:Vector="0, 0, 0", Rotation:Rotator, bTransient:bool) -> (then, ReturnValue:BP Yoke Hop Unkind Spline Editor Object Reference)
   > n4 = GeneratePoints!(execute, self:BP Yoke Hop Unkind Spline Editor Object Reference=@n3) -> (then, OutputPoints:"Array of Float (double-precision)s")
   > n5 = Sequence()
   > n6 = LoadAndGetAssetObject!(execute, self:EUW Custom Asset Picker Object Reference=Get DataAssetPicker() -> (DataAssetPicker:EUW Custom Asset Picker Object Reference)) -> (then, asset:Object Reference, HasSelection:bool)
   > Branch(Condition=@n6.HasSelection)
   > n7 = Cast To DataAsset_YokeHopUnkind_ControlPoints!(execute, Object=@n6) -> (then, CastFailed, AsData Asset Yoke Hop Unkind Control Points:Data Asset Yoke Hop Unkind Control Points Object Reference) {is=title K2Node_DynamicCast}
   > Set ControlPointRadiusScales!(
    execute
    ControlPointRadiusScales:"Array of Float (double-precision)s"=@n4
    self:Data Asset Yoke Hop Unkind Control Points Object Reference=@n7
   ) -> (then, Output_Get:"Array of Float (double-precision)s")
   > SaveLoadedAsset(AssetToSave=@n6, bOnlyIfIsDirty=true)
   Destroy Actor(execute=@n5.then_1, self=@n3)
   "On Clicked (PickSourceButton)"() {is=title K2Node_ComponentBoundEvent}
   > n8 = GetSelectedAssets()
   > n9 = For Each Loop(Array:Array of Object References=@n8) -> (Array Element:Object Reference)
   > n10 = Set SelectedObject(SelectedObject:Object Reference=@n9) -> (Output_Get:Object Reference)
   > n11 = GetSplineEditorClassFromSelectedObject(Object=@n10)
   > n12 = Set SplineEditorClass(SplineEditorClass:BP Yoke Hop Unkind Spline Editor Class Reference=@n11) -> (Output_Get:BP Yoke Hop Unkind Spline Editor Class Reference)
   > Set Text(
    Text=Format Text!(Format="LOCTEXT(\"Current Source: {SourceName}\")", SourceName:String=Get Class Display Name(Class=@n12)) -> (Result)
    self=Get SplineSourceText() -> (SplineSourceText:Text Object Reference)
   )
   Get SplineEditorClass() -> (SplineEditorClass:BP Yoke Hop Unkind Spline Editor Class Reference)
  ]}
 GetSplineEditorClassFromSelectedObject {nodes [
   n1 = GetSplineEditorClassFromSelectedObject!() -> (then, Object:Object Reference) {is=title K2Node_FunctionEntry}
   > n2 = GetBlueprintAsset(Object=@n1)
   > Is Valid(InputObject=@n2)
   > n3 = GeneratedClass!(execute, BlueprintObj:Blueprint Object Reference=@n2) -> (then, ReturnValue:Object Class Reference)
   > n4 = Cast To BP_YokeHopUnkind_SplineEditor Class!(execute:Exec, Class:Object Class Reference=@n3) -> (then:Exec, CastFailed:Exec, AsBP Yoke Hop Unkind Spline Editor:BP Yoke Hop Unkind Spline Editor Class Reference) {is=title K2Node_ClassDynamicCast}
   Return Node!(execute=[@n4.CastFailed, @n4.then], AsBP Yoke Hop Unkind Spline Editor:BP Yoke Hop Unkind Spline Editor Class Reference=@n4)
  ]}
]
variables [
 NewVar {type=Class, default=null}
 SelectedObject {type=Object, default=null}
 SplineEditorClass {type=Class, default=null}
]
widgets {
 info {rootWidgetClass=&StackBox, widgetCount=7}
 widgets [
  StackBox_0 {widgetClassPath=&StackBox, widget=StackBox_0, properties {orientation=Orient_Vertical}}
  SplineSourceText {
   widgetClassPath=&TextBlock
   bIsVariable=true
   parent=StackBox_0
   widget=SplineSourceText
   properties {
    text=Current Source: NONE
    font {
     bForceMonospaced=false
     bMaterialIsStencil=false
     fontMaterial=null
     fontObject=&Roboto
     letterSpacing=0
     monospacedWidth=1
     size=12
     skewAmount=0
     typefaceFontName=Bold
     outlineSettings {
      bApplyOutlineToDropShadows=false
      bMiteredCorners=false
      bSeparateFillAlpha=false
      outlineMaterial=null
      outlineSize=0
      outlineColor=<rgba 0 0 0 1>
     }
    }
   }
   slot {class=StackBoxSlot}
  }
  PickSourceButton {
   widgetClassPath=&EditorUtilityButton
   parent=StackBox_0
   widget=PickSourceButton
   properties {toolTipText="Select the spline editor class you want to use in the content browser, then click this button to set it as the active source"}
   slot {class=StackBoxSlot, properties {horizontalAlignment=HAlign_Center, padding=<ltrb 20 20 20 20>}}
  }
  TextBlock_64 {
   widgetClassPath=&TextBlock
   parent=PickSourceButton
   widget=TextBlock_64
   properties {
    text=Set Spline Editor Source
    font {
     bForceMonospaced=false
     bMaterialIsStencil=false
     fontMaterial=null
     fontObject=&Roboto
     letterSpacing=0
     monospacedWidth=1
     size=12
     skewAmount=0
     typefaceFontName=Bold
     outlineSettings {
      bApplyOutlineToDropShadows=false
      bMiteredCorners=false
      bSeparateFillAlpha=false
      outlineMaterial=null
      outlineSize=0
      outlineColor=<rgba 0 0 0 1>
     }
    }
   }
   slot {class=ButtonSlot}
  }
  DataAssetPicker {
   widgetClassPath=/Game/NeverCook/EditorUtilities/CustomAssetPicker/EUW_CustomAssetPicker
   bIsVariable=true
   parent=StackBox_0
   widget=DataAssetPicker
   properties {label=Output Asset, classFilters [&/DataAsset_YokeHopUnkind_ControlPoints.DataAsset_YokeHopUnkind_ControlPoints_C]}
   slot {class=StackBoxSlot}
  }
  OutputButton {
   widgetClassPath=&EditorUtilityButton
   parent=StackBox_0
   widget=OutputButton
   properties {toolTipText="Select the spline editor class you want to use in the content browser, then click this button to set it as the active source"}
   slot {class=StackBoxSlot, properties {horizontalAlignment=HAlign_Center, padding=<ltrb 20 20 20 20>}}
  }
  TextBlock {
   widgetClassPath=&TextBlock
   parent=OutputButton
   widget=TextBlock
   properties {
    text=Write Control Points to Selected Asset
    font {
     bForceMonospaced=false
     bMaterialIsStencil=false
     fontMaterial=null
     fontObject=&Roboto
     letterSpacing=0
     monospacedWidth=1
     size=12
     skewAmount=0
     typefaceFontName=Bold
     outlineSettings {
      bApplyOutlineToDropShadows=false
      bMiteredCorners=false
      bSeparateFillAlpha=false
      outlineMaterial=null
      outlineSize=0
      outlineColor=<rgba 0 0 0 1>
     }
    }
   }
   slot {class=ButtonSlot}
  }
 ]
}
